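; DUI_SunRoseFlyingGameplay is a Blueprint in the UI folder of the SunRose flying gameplay wings item, apparently a dynamic UI director going by its prefix.
; It declares BeginPlay, Tick, a construction script, and OnNewPawn and OnWeaponEquipped functions that receive new and previous pawns or weapons.
; All are empty here, so the graphs add no logic of their own.
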(event EventBeginPlay)

(event EventTick
  (params (DeltaSeconds :type Float%20%28single-precision%29)))

(fn OnNewPawn
  (params (OldPawn :type Pawn%20Object%20Reference) (NewPawn :type Pawn%20Object%20Reference)))

(fn OnWeaponEquipped
  (params (NewWeapon :type Fort%20Weapon%20Object%20Reference) (PrevWeapon :type Fort%20Weapon%20Object%20Reference)))

(fn ConstructionScript)
