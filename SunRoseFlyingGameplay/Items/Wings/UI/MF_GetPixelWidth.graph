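# Material function in the Wings item UI folder that returns the width of one screen pixel in UV units for a displayed material.
# The optional UVs input defaults to the 9-Slice UV from GetUserInterfaceUV, allowing for varying pixel density.
# It reduces the UVs to their length, then outputs the magnitude of that value's DDX and DDY screen derivatives as PixelWidth.
# Both length operations use small Custom HLSL nodes.
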MaterialFunction /SunRoseFlyingGameplay/Items/Wings/UI/MF_GetPixelWidth
graph {
 properties {
  description="Retrives the screen pixel width of a displayed material, optionally modulated by supplied UVs."
  libraryCategoriesText [User Interface]
 }
 nodes [
  MaterialExpressionFunctionInput_0 = FunctionInput(Preview=MaterialFunctionCall() {materialFunction=&GetUserInterfaceUV}.9-Slice UV) {
   bUsePreviewValueAsDefault=true
   description="The UVs to retrieve pixel width for (useful in cases where pixel density varies throughout the UV map)."
   inputName=UVs
   inputType=FunctionInput_Vector2
  }
  n1 = Custom!(Vec=@MaterialExpressionFunctionInput_0) {
   code="return length(Vec);"
   description=Length
   outputType=CMOT_Float1
   inputs [{inputName=Vec, input {expression=&self:MaterialExpressionFunctionInput_0}}]
  }
  MaterialExpressionAppendVector_0 = AppendVector(A=DDX(Value=@n1), B=DDY(Value=@n1))
  n2 = Custom(Vec=@MaterialExpressionAppendVector_0) {
   code="return length(Vec);"
   description=Length
   outputType=CMOT_Float1
   inputs [{inputName=Vec, input {expression=&self:MaterialExpressionAppendVector_0}}]
  }
  FunctionOutput(None=@n2) {outputName=PixelWidth}
 ]
}
