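# Translucent UI material for the flying Wings reticle, drawing a dot, dive-angle ticks, anchor ticks and dive brackets.
# One texture packs dot, brackets and ticks into RGB, with symmetry UV functions mirroring ticks and brackets.
# Scalar parameters toggle and fade elements, slide ticks, and animate bracket offset and rotation under a circle mask.
# Two smoothstep thresholds build sharp and glow sums, blended into the final reticle mask.
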Material /SunRoseFlyingGameplay/Items/Wings/UI/M_UI_UXReticle
graph {
 properties {blendMode=BLEND_Translucent, materialDomain=MD_UI, pixelDepthOffsetMode=PDOM_Legacy, usedShadingModels=Default Lit}
 nodes [
  LinearInterpolate(A=Constant4Vector() {constant=<rgba 0 0 0 1>}.RGBA, B=Constant4Vector() {constant=<rgba 1 1 1 0>}.RGBA, Alpha=NamedRerouteUsage())
  Multiply(
   A=NamedRerouteUsage()
   B=LinearInterpolate(Alpha=ScalarParameter() {defaultValue=1, parameterName=Material Opacity}) {constA=0.25, constB=0.85}
  )
  n1 = LinearInterpolate(B=NamedRerouteUsage(), Alpha=Ceil(None=ScalarParameter() {defaultValue=1, parameterName=useDiveTicks}))
  n2 = Multiply(A=SmoothStep(Value=@n1) {constMax=0.5, constMin=0.25}, B=ScalarParameter() {defaultValue=0.5, parameterName=DA Tick Opacity, sliderMax=1})
  n3 = LinearInterpolate(B=NamedRerouteUsage(), Alpha=Ceil(None=ScalarParameter() {defaultValue=1, parameterName=useAnchorTicks}))
  n4 = Multiply(A=SmoothStep(Value=@n3) {constMax=0.5, constMin=0.25}, B=ScalarParameter() {defaultValue=1, parameterName=Anchor Tick Opacity, sliderMax=1})
  n5 = LinearInterpolate(B=NamedRerouteUsage(), Alpha=Ceil(None=ScalarParameter() {defaultValue=1, parameterName=useDiveBrackets}))
  n6 = SmoothStep(Value=@n5) {constMax=0.5, constMin=0.25}
  n7 = ScalarParameter() {defaultValue=0.5, parameterName=Dive Bracket Opacity, sliderMax=1}
  n8 = NamedRerouteUsage()
  n9 = Add(A=Add(A=Add(A=@n2, B=@n4), B=Multiply(A=@n6, B=@n7)), B=SmoothStep(Value=@n8) {constMax=0.5, constMin=0.25})
  n10 = Multiply(A=SmoothStep(Value=@n1) {constMax=0.4, constMin=0.1}, B=ScalarParameter() {defaultValue=0.5, parameterName=DA Tick Opacity, sliderMax=1})
  n11 = Multiply(A=SmoothStep(Value=@n3) {constMax=0.4, constMin=0.1}, B=ScalarParameter() {defaultValue=1, parameterName=Anchor Tick Opacity, sliderMax=1})
  n12 = SmoothStep(Value=@n5) {constMax=0.4, constMin=0.1}
  n13 = ScalarParameter() {defaultValue=0.5, parameterName=Dive Bracket Opacity, sliderMax=1}
  n14 = Add(A=Add(A=Add(A=@n10, B=@n11), B=Multiply(A=@n12, B=@n13)), B=SmoothStep(Value=@n8) {constMax=0.4, constMin=0.1})
  NamedRerouteDeclaration(None=LinearInterpolate(
    A=NamedRerouteDeclaration(None=@n9) {name=Reticle Sharp, nodeColor=<rgba 1 1 1 1>}
    B=NamedRerouteDeclaration(None=@n14) {name=Reticle Glow, nodeColor=<rgba 1 1 1 1>}
   ) {constAlpha=0.4}) {name=Reticle Final, nodeColor=<rgba 1 1 1 1>}
  n15 = TextureSample(UVs=MaterialFunctionCall() {materialFunction=&GetUserInterfaceUV}.Normalized UV) {samplerSource=SSM_Clamp_WorldGroupSettings, texture=/SunRoseFlyingGameplay/Items/Wings/UI/Images/T_UI_Reticle_OlympusWings_ReticleAndPips}
  NamedRerouteDeclaration(None=ComponentMask(None=@n15.RGB) {r=true}) {name=Dot, nodeColor=<rgba 1 1 1 1>}
  n16 = ScalarParameter() {defaultValue=-0.475, parameterName=xOffsetVal}
  n17 = LinearInterpolate(A=@n16, Alpha=ScalarParameter() {parameterName=xOffset Controller, sliderMax=1}) {constA=-0.475, constB=-0.75}
  n18 = LinearInterpolate(Alpha=ScalarParameter() {parameterName=yOffsetController, sliderMax=1}) {constA=-0.2, constB=0}
  n19 = Constant2Vector() {g=0.75, r=1.5}
  n20 = MaterialFunctionCall() {materialFunction=&GetUserInterfaceUV}
  n21 = TextureSample(UVs=MaterialFunctionCall!(UV=@n20.Normalized UV, Offset=AppendVector(A=@n17, B=@n18), Rotation, Scale=@n19, MirrorY) {materialFunction=/FortUILibrary/MaterialFunctions/Basics/MF_UI_Symmetry_2x1_DEPRECATED}.Result) {samplerSource=SSM_Clamp_WorldGroupSettings, texture=&T_UI_Reticle_OlympusWings_ReticleAndPips}
  NamedRerouteDeclaration(None=ComponentMask(None=@n21.RGB) {b=true}) {name=Dive Angle Ticks, nodeColor=<rgba 1 1 1 1>}
  n22 = Constant()
  n23 = TextureSample(UVs=MaterialFunctionCall(UV=@n20.Normalized UV, Offset=AppendVector(A=@n17, B=@n22), Scale=@n19) {materialFunction=&MF_UI_Symmetry_2x1_DEPRECATED}.Result) {samplerSource=SSM_Clamp_WorldGroupSettings, texture=&T_UI_Reticle_OlympusWings_ReticleAndPips}
  NamedRerouteDeclaration(None=ComponentMask(None=@n23.RGB) {b=true}) {name=Anchor Ticks, nodeColor=<rgba 1 1 1 1>}
  n24 = LinearInterpolate(Alpha=ScalarParameter() {defaultValue=1, parameterName=Rotation Controller, sliderMax=1}) {constB=0.25}
  n25 = LinearInterpolate(A=Constant() {r=-0.375}, B=@n16, Alpha=ScalarParameter() {defaultValue=1, parameterName=BracketAnimController, sliderMax=1}) {constB=-0.475}
  n26 = Constant()
  n27 = TextureSample(UVs=MaterialFunctionCall(
    UV=MaterialFunctionCall(UVs=@n20.Normalized UV, "Rotation Angle (0-1)"=@n24) {materialFunction=&CustomRotator}.Rotated Values
    Offset=AppendVector(A=@n25, B=@n26)
    Scale=@n19
   ) {materialFunction=&MF_UI_Symmetry_2x1_DEPRECATED}.Result) {samplerSource=SSM_Clamp_WorldGroupSettings, texture=&T_UI_Reticle_OlympusWings_ReticleAndPips}
  n28 = Constant() {r=1}
  NamedRerouteDeclaration(None=Multiply(
    A=ComponentMask(None=@n27.RGB) {g=true}
    B=MaterialFunctionCall(Size=@n28) {desc=Mask for Brackets, materialFunction=&MF_UI_Circle}.Fill
   )) {name=Dive Brackets, nodeColor=<rgba 1 1 1 1>}
 ]
}
