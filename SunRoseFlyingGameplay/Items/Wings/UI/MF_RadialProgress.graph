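# Material function for UI that outputs a radial progress sector mask, for timers or bars in the wings item's UI.
# It rotates UVs (9-slice UI UVs by default) by StartValue and converts them to a normalized angle with a fast arctangent.
# The sign of Range sets the sweep direction, and its magnitude scaled by Progress sets the sweep size.
# MF_AntiAliasing thresholds the summed angle and sweep, with output forced to zero unless Progress is positive.
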MaterialFunction /SunRoseFlyingGameplay/Items/Wings/UI/MF_RadialProgress
graph {
 properties {
  description=Generates a radial progress sector. Useful for masking textures to make radial progress timers or bars.
  libraryCategoriesText [User Interface]
 }
 nodes [
  n1 = Constant()
  n2 = FunctionInput() {
   bUsePreviewValueAsDefault=true
   description="The range (in degrees) of the progress bar. Positive values progress clockwise, negative values go counter-clockwise."
   inputName=Range
   inputType=FunctionInput_Scalar
   sortPriority=1
   previewValue=<360 0 0 1>
  }
  n3 = FunctionInput(Preview=MaterialFunctionCall() {materialFunction=&GetUserInterfaceUV}.9-Slice UV) {bUsePreviewValueAsDefault=true, description=Optional custom UV mapping., inputName=UVs, inputType=FunctionInput_Vector2, sortPriority=3}
  n4 = FunctionInput() {
   bUsePreviewValueAsDefault=true
   description="The starting direction in degrees of the meter (0-360) where 0 is up, 90 is right, 180 is down, and 270 is left."
   inputName=StartValue
   inputType=FunctionInput_Scalar
  }
  n5 = Subtract(A=Constant() {r=0.5}, B=Rotator(Coordinate=@n3, Time=@n4) {speed=-0.0175})
  n6 = ComponentMask(None=@n5) {r=true}
  n7 = ComponentMask(None=@n5) {g=true}
  n8 = Frac(None=Divide(A=Arctangent2Fast(Y=@n6, X=@n7), B=Constant() {r=6.2832}))
  n9 = Abs(None=@n2)
  n10 = FunctionInput(Preview=Constant() {r=0.6}) {description="The current progress (0-1).", inputName=Progress, inputType=FunctionInput_Scalar, sortPriority=2}
  n11 = Add(A=LinearInterpolate(B=Multiply(A=@n9) {constB=0.0028}, Alpha=@n10), B=If(A=@n2, "A > B"=@n8, "A == B"=@n8, "A < B"=OneMinus(None=@n8)))
  n12 = Constant() {r=1}
  FunctionOutput(None=If(
    A=@n10
    B=@n1
    "A > B"=MaterialFunctionCall!(Value=@n11, Threshold=@n12, Smoothing) {materialFunction=&/MF_AntiAliasing}.EitherSide
    "A == B"=@n1
    "A < B"=@n1
   ))
 ]
}
