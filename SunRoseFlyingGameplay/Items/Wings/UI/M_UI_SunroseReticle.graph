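# UI material for the Sunrose wings flying reticle, showing a fuel gauge driven by Progress, with a Fuel Opacity parameter.
# A texture supplies frame (red) and bar (green) channels; a radial progress function on mirrored UVs, snapped to full near the top, builds the tank fill.
# Smoothsteps derive solid and glow layers for frame and tank.
# Tank color is white above half, then pulses between white and yellow, or red when low.
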Material /SunRoseFlyingGameplay/Items/Wings/UI/M_UI_SunroseReticle
graph {
 properties {blendMode=BLEND_Translucent, materialDomain=MD_UI, pixelDepthOffsetMode=PDOM_Legacy, usedShadingModels=Default Lit}
 nodes [
  Multiply(A=NamedRerouteUsage(), B=NamedRerouteUsage())
  Multiply(A=NamedRerouteUsage(), B=LinearInterpolate(Alpha=ScalarParameter() {defaultValue=1, parameterName=Fuel Opacity, sliderMax=1}) {constB=0.8}) {constB=0.5}
  n1 = Constant() {desc=Range, r=-57}
  n2 = ScalarParameter() {defaultValue=1, parameterName=Progress, sliderMax=1}
  n3 = Constant() {r=1}
  n4 = If(A=@n2, "A > B"=@n3, "A == B"=@n3, "A < B"=@n2) {constB=0.94, equalsThreshold=0.001}
  n5 = MaterialFunctionCall!(
   UV=MaterialFunctionCall() {materialFunction=&GetUserInterfaceUV}.Normalized UV
   Offset=Constant2Vector() {r=-0.5}
   Rotation
   Scale=Constant2Vector() {g=0.5, r=1}
   MirrorY
  ) {materialFunction=/FortUILibrary/MaterialFunctions/Basics/MF_UI_Symmetry_2x1_DEPRECATED}
  NamedRerouteDeclaration(None=Multiply(
    A=MaterialFunctionCall!(StartValue, Range=@n1, Progress=@n4, UVs=@n5.Result) {materialFunction=&/MF_RadialProgress}.Result
    B=NamedRerouteUsage()
   )) {name=CurrentTank, nodeColor=<rgba 1 1 1 1>}
  n6 = NamedRerouteUsage()
  n7 = SmoothStep(Value=@n6) {constMax=0.5, constMin=0.3}
  n8 = NamedRerouteUsage()
  n9 = SmoothStep(Value=@n8) {constMax=0.5, constMin=0.25}
  n10 = Add(
   A=Multiply(A=NamedRerouteDeclaration(None=@n7) {name=CurrentTankSolid, nodeColor=<rgba 1 1 1 1>}, B=NamedRerouteUsage())
   B=Multiply(A=NamedRerouteDeclaration(None=@n9) {name=FrameSolid, nodeColor=<rgba 1 1 1 1>}) {constB=0.75}
  )
  n11 = Add(
   A=NamedRerouteDeclaration(None=SmoothStep(Value=@n8) {constMax=0.4, constMin=0.01}) {name=FrameGlow, nodeColor=<rgba 1 1 1 1>}
   B=NamedRerouteDeclaration(None=SmoothStep(Value=@n6) {constMax=0.4, constMin=0.01}) {name=CurrentTankGlow, nodeColor=<rgba 1 1 1 1>}
  )
  NamedRerouteDeclaration(None=LinearInterpolate(
    A=NamedRerouteDeclaration(None=@n10) {name=RetSolid, nodeColor=<rgba 1 1 1 1>}
    B=NamedRerouteDeclaration(None=@n11) {name=RetGlow, nodeColor=<rgba 1 1 1 1>}
   )) {name=RetFin, nodeColor=<rgba 1 1 1 1>}
  n12 = Constant3Vector() {constant=<rgba 1 0.2413 0.2515 1>}
  n13 = Multiply(A=Time()) {constB=1.5}
  n14 = LinearInterpolate(
   A=If(
    A=ScalarParameter() {defaultValue=1, parameterName=Progress, sliderMax=1}
    "A > B"=Constant3Vector() {constant=<rgba 1 0.6685 0.2413 1>}
    "A == B"=@n12
    "A < B"=@n12
   ) {constB=0.2}
   B=Constant() {r=1}
   Alpha=ConstantBiasScale(None=Sine(None=@n13))
  )
  NamedRerouteDeclaration(None=If(A=ScalarParameter() {defaultValue=1, parameterName=Progress, sliderMax=1}, "A > B"=Constant() {r=1}, "A == B"=@n14, "A < B"=@n14) {constB=0.5}) {name=TankColors, nodeColor=<rgba 1 1 1 1>}
  n15 = MaterialFunctionCall() {materialFunction=&GetUserInterfaceUV}
  n16 = Constant2Vector() {g=0.3}
  n17 = TextureSample(UVs=MaterialFunctionCall(UVs=Add(A=@n15.Normalized UV, B=@n16) {constB=0.25}, Texture Scale=Constant2Vector() {g=0.8, r=0.8}) {materialFunction=&ScaleUVsByCenter}.UVs) {samplerSource=SSM_Clamp_WorldGroupSettings, texture=/SunRoseFlyingGameplay/Items/Wings/UI/Images/T_UI_Reticle_OlympusWings_AmmoBar}
  NamedRerouteDeclaration(None=ComponentMask(None=@n17.RGB) {g=true}) {name=T_Bar, nodeColor=<rgba 1 1 1 1>}
  NamedRerouteDeclaration(None=ComponentMask(None=@n17.RGB) {r=true}) {name=T_Frame, nodeColor=<rgba 1 1 1 1>}
 ]
}
