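# MetaSound Source for the SunRose wings low-energy UI warning, pulsing in stereo while Progress is low and Hidden is false.
# A 1 Hz LFO feeds threshold compares and a self-closing trigger gate to make one pulse per cycle, gated by Progress between 0 and 0.4.
# Each pulse plays both red and yellow wave players.
# Trigger routes select the pitched-up red sound below 0.2 Progress, otherwise yellow.
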MetaSoundSource /SunRoseFlyingGameplay/Sounds/MSS_SunRose_Wings_UI_LowEnergyPulse
inputs [
 Hidden {type=Bool}
 Progress {type=Float}
 Sound.RedPulse {type=WaveAsset:Array}
 Sound.YellowPulse {type=WaveAsset:Array}
 UE.Source.OnPlay {type=Trigger}
]
interfaces [UE.OutputFormat.Stereo:1.0, UE.Source:1.1]
outputs [
 "UE.OutputFormat.Stereo.Audio:0" {type=Audio}
 "UE.OutputFormat.Stereo.Audio:1" {type=Audio}
]
pages [Default {
  input_defaults {
   Hidden=null
   Progress=1
   UE.Source.OnPlay=null
   Sound.RedPulse [/SunRoseFlyingGameplay/Sounds/wav/FN_S29_OW_Energy_UI_Red]
   Sound.YellowPulse [/SunRoseFlyingGameplay/Sounds/wav/FN_S29_OW_Energy_UI_Yellow]
  }
  nodes [
   n1 = UE.Reroute(Value=^Progress)
   n2 = UE.LFO:Audio() {also {defaults {Frequency=1, Min Value=0}}}
   TriggerOnValueChange_1 = UE.Trigger On Value Change:Float(Value=@n2.Out)
   n3 = VariableDeferredAccessor.Trigger(Variable=InitVariable.Trigger().Variable)
   TriggerControl_0 = UE.Trigger Control(
    Trigger In=TriggerCompare.Clamp:Float(A=@n2.Out, Compare=@TriggerOnValueChange_1.Trigger) {also {defaults {B=0.5, Type=3}}}.True
    Open=TriggerCompare.Clamp:Float(A=@n2.Out, Compare=@TriggerOnValueChange_1.Trigger) {also {defaults {B=0.4, Type=2}}}.True
    Close=@n3.Value
   )
   n4 = VariableMutator.Trigger(Value=@TriggerControl_0.Trigger Out, Variable=@n3.Variable)
   n5 = TriggerCompare.Clamp:Float(
    A=^Progress
    Compare=TriggerCompare.Clamp:Bool(A=^Hidden, Compare=VariableAccessor.Trigger(Variable=@n4.Variable).Value) {also {defaults {B="<MetaSoundBuilderResult.SUCCEEDED: 0>"}}}.False
   ) {also {defaults {B=0.4, Type=2}}}
   n6 = VariableAccessor.Trigger(Variable=VariableMutator.Trigger(
     Variable=InitVariable.Trigger().Variable
     Value=TriggerCompare.Clamp:Float(A=@n1.Value, Compare=@n5.True) {also {defaults {B=0, Type=3}}}.True
    ).Variable)
   n7 = &MSP_Audio_WavePlayer_Deluxe_Stereo(Wave Asset=^Sound.RedPulse, Play Trigger=@n6.Value) {also {defaults {
      Enable Shared Random State="<MetaSoundBuilderResult.SUCCEEDED: 0>"
      "Pitch Shift Global (Semitones)"=1.2
      "Random Pitch Max (Semitones)"=0
      "Random Pitch Min (Semitones)"=0
     }}}
   n8 = &MSP_Audio_WavePlayer_Deluxe_Stereo(Wave Asset=^Sound.YellowPulse, Play Trigger=VariableAccessor.Trigger(Variable=@n6.Variable).Value) {also {defaults {Enable Shared Random State="<MetaSoundBuilderResult.SUCCEEDED: 0>", "Random Pitch Max (Semitones)"=0, "Random Pitch Min (Semitones)"=0}}}
   TriggerOnValueChange_0 = UE.Trigger On Value Change:Float(Value=^Progress)
   n9 = TriggerCompare.Clamp:Float(A=^Progress, Compare=@TriggerOnValueChange_0.Trigger) {also {defaults {B=0.2, Type=2}}}
   n10 = VariableAccessor.Trigger(Variable=VariableMutator.Trigger(Variable=InitVariable.Trigger().Variable, Value=@n9.True).Variable)
   n11 = VariableAccessor.Trigger(Variable=VariableMutator.Trigger(Variable=InitVariable.Trigger().Variable, Value=@n9.False).Variable)
   TriggerRouteTriggerRouteAudio2_0 = "TriggerRoute.Trigger Route (Audio, 2):Audio"(Value 0=@n7.Out L, Value 1=@n8.Out L, Set 0=@n10.Value, Set 1=@n11.Value)
   TriggerRouteTriggerRouteAudio2_1 = "TriggerRoute.Trigger Route (Audio, 2):Audio"(
    Value 0=@n7.Out R
    Value 1=@n8.Out R
    Set 0=VariableAccessor.Trigger(Variable=@n10.Variable).Value
    Set 1=VariableAccessor.Trigger(Variable=@n11.Variable).Value
   )
  ]
  links [
   {dst="@UE.OutputFormat.Stereo.Audio:0", src=TriggerRouteTriggerRouteAudio2_0.Value}
   {dst="@UE.OutputFormat.Stereo.Audio:1", src=TriggerRouteTriggerRouteAudio2_1.Value}
  ]
 }]
properties {
 assetImportData=&self:AssetImportData
 bLooping=true
 bOverrideConcurrency=true
 duration=10000
 numChannels=2
 outputFormat=Stereo
 sampleRate=48000
 soundClassObject=&Fort_UI_InGame_NonCritical
 concurrencyOverrides {bLimitToOwner=true, maxCount=2, resolutionRule=PreventNew, retriggerTime=0.4, volumeScale=0.5}
}
