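# Weapon reticle extension widget for the SunRose wings, showing fuel, dive aim and leap cooldown.
# Local ammo changes update fuel and reserve material parameters, a brief hint, and a low-energy sound loop's Progress parameter.
# Async gameplay-tag waits on Leaping, Flying and cooldown drive visibility, audio hiding and cooldown animation.
# Tick maps look pitch against a curve-table crash clamp to fade ticks and play lock/unlock animations at the limit.
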WidgetBlueprint /SunRoseFlyingGameplay/Items/Wings/UI/WBP_Reticle_SunRoseFlying
animations [
 OnCooldown {
  binding_count=3
  length_seconds=0.2333
  object=&self:OnCooldown
  sequence {
   bindings [
    Overlay_15 {
     id=#1
     possessed_class=Overlay
     tracks [{
       type=MovieScene2DTransformTrack
       sections [{
         type=MovieScene2DTransformSection
         end_frame=7
         start_frame=0
         channels [
          Translation.X {type=MovieSceneScriptingFloatChannel, const [0,1]}
          Translation.Y {type=MovieSceneScriptingFloatChannel, const [0,1]}
          Angle {type=MovieSceneScriptingFloatChannel, const [0,1]}
          Scale.X {type=MovieSceneScriptingFloatChannel, keys [[0,1],[6,0.85]]}
          Scale.Y {type=MovieSceneScriptingFloatChannel, keys [[0,1],[6,0.85]]}
          Shear.X {type=MovieSceneScriptingFloatChannel, const [0,1]}
          Shear.Y {type=MovieSceneScriptingFloatChannel, const [0,1]}
         ]
        }]
      }]
    }
    centerRet {
     id=#2
     possessed_class=Image
     tracks [
      {
       type=MovieSceneWidgetMaterialTrack
       sections [{
         type=MovieSceneParameterSection
         end_frame=6
         start_frame=0
         channels [
          Anchor Tick Opacity {type=MovieSceneScriptingFloatChannel, const [0,1]}
          Dive Bracket Opacity {type=MovieSceneScriptingFloatChannel, keys [[0,0.5],[3,0]]}
          Material Opacity {type=MovieSceneScriptingFloatChannel, keys [[0,1],[5,0.5]]}
          BracketAnimController {type=MovieSceneScriptingFloatChannel, keys [[0,1],[5,0]]}
          DA Tick Opacity {type=MovieSceneScriptingFloatChannel, const [0,1]}
         ]
        }]
      }
      {
       type=MovieScene2DTransformTrack
       sections [{
         type=MovieScene2DTransformSection
         end_frame=5
         start_frame=0
         channels [
          Translation.X {type=MovieSceneScriptingFloatChannel, const [0,1]}
          Translation.Y {type=MovieSceneScriptingFloatChannel, const [0,1]}
          Angle {type=MovieSceneScriptingFloatChannel}
          Scale.X {type=MovieSceneScriptingFloatChannel, const [1,1]}
          Scale.Y {type=MovieSceneScriptingFloatChannel, const [1,1]}
          Shear.X {type=MovieSceneScriptingFloatChannel, const [0,1]}
          Shear.Y {type=MovieSceneScriptingFloatChannel, const [0,1]}
         ]
        }]
      }
     ]
    }
    fuelRet {
     id=#3
     possessed_class=Image
     tracks [{
       type=MovieSceneWidgetMaterialTrack
       sections [{
         type=MovieSceneParameterSection
         end_frame=6
         start_frame=0
         channels [Fuel Opacity {type=MovieSceneScriptingFloatChannel, keys [[0,1],[5,0.35]]}]
        }]
      }]
    }
   ]
   display_rate {numerator=30}
  }
 }
 OnFlight {
  binding_count=1
  length_seconds=5
  object=&self:OnFlight
  sequence {
   bindings [centerRet {
     id=#4
     possessed_class=Image
     tracks [{
       type=MovieSceneWidgetMaterialTrack
       sections [{
         type=MovieSceneParameterSection
         end_frame=150
         start_frame=0
         channels [
          DA Tick Opacity {type=MovieSceneScriptingFloatChannel, const [0,1]}
          Anchor Tick Opacity {type=MovieSceneScriptingFloatChannel, keys [[0,0],[4,1]]}
          Dive Bracket Opacity {type=MovieSceneScriptingFloatChannel, const [0,1]}
          xOffset Controller {type=MovieSceneScriptingFloatChannel, keys [[0,1],[6,0]]}
         ]
        }]
      }]
    }]
   display_rate {numerator=30}
  }
 }
 OnGround {binding_count=0, length_seconds=5, object=&self:OnGround, sequence {bindings []}}
 AmmoHint {
  binding_count=1
  length_seconds=2.0333
  object=&self:AmmoHint
  sequence {
   bindings [fuelRet {
     id=#5
     possessed_class=Image
     tracks [{
       type=MovieSceneWidgetMaterialTrack
       sections [{
         type=MovieSceneParameterSection
         end_frame=61
         start_frame=0
         channels [Fuel Opacity {type=MovieSceneScriptingFloatChannel, keys [[0,0],[10,1],[50,1],[60,0]]}]
        }]
      }]
    }]
   display_rate {numerator=30}
  }
 }
 LockedTicks {
  binding_count=1
  length_seconds=0.1
  object=&self:LockedTicks
  sequence {bindings [centerRet {
     id=#6
     possessed_class=Image
     tracks [{
       type=MovieSceneWidgetMaterialTrack
       sections [{
         type=MovieSceneParameterSection
         end_frame=2
         start_frame=0
         channels [xOffset Controller {type=MovieSceneScriptingFloatChannel, keys [[0,-1],[2,1.0291]]}]
        }]
      }]
    }]}
 }
 LockedBrackets {
  binding_count=1
  length_seconds=0.2
  object=&self:LockedBrackets
  sequence {bindings [centerRet {
     id=#7
     possessed_class=Image
     tracks [{
       type=MovieSceneWidgetMaterialTrack
       sections [{
         type=MovieSceneParameterSection
         end_frame=4
         start_frame=0
         channels [
          BracketAnimController {type=MovieSceneScriptingFloatChannel, keys [[0,1],[2,1],[3,0.453],[4,0.26]]}
          Dive Bracket Opacity {type=MovieSceneScriptingFloatChannel, keys [[0,0],[1,0.5],[4,0.5]]}
         ]
        }]
      }]
    }]}
 }
 UnlockBrackets {
  binding_count=1
  length_seconds=0.25
  object=&self:UnlockBrackets
  sequence {bindings [centerRet {
     id=#8
     possessed_class=Image
     tracks [{
       type=MovieSceneWidgetMaterialTrack
       sections [{
         type=MovieSceneParameterSection
         end_frame=5
         start_frame=0
         channels [Dive Bracket Opacity {type=MovieSceneScriptingFloatChannel, keys [[0,0.5],[5,0]]}]
        }]
      }]
    }]}
 }
]
graphs [
 EventGraph {nodes [
   n1 = UpdateAmmo!() -> (OutputDelegate, then, LocalCount:int, LocalRemaining:int) {is=title K2Node_CustomEvent}
   > Show Ammo Hint()
   > n2 = Sequence!(execute) -> (then_0, then_1, then_2:Exec)
   n3 = float * float(A:int=@n1.LocalCount, B=0.01) {is=title K2Node_PromotableOperator}
   > SetScalarParameterValue(self=Get FuelMat() -> (FuelMat:Material Instance Dynamic Object Reference), ParameterName=Progress, Value=@n3)
   > n4 = Get!(execute:Exec) -> (then:Exec, else:Exec, ReticleSoundLoop:Audio Component Object Reference)
   > Set Float Parameter!(execute, self:Audio Parameter Controller Interface Interface=@n4, InName:Name=Progress, InFloat:float=@n3) -> (then) {is=title K2Node_CallFunction}
   n5 = int + int(A=int - int(A=100, B=@n1.LocalCount), B=@n1.LocalRemaining)
   SetScalarParameterValue(
    execute=@n2.then_1
    self=Get FuelMat()
    ParameterName=ReservesProgress
    Value=float * float(A:double=float - float(A:int=@n5, B:int=100), B=0.005)
   )
   HandleASCReady!() -> (OutputDelegate, then, AbilitySystemComponent:Fort Ability System Component Object Reference, AffectedPawn:Fort Player Pawn Object Reference) {is=title K2Node_CustomEvent}
   n6 = Get Gameplay Tag - Leaping() -> (Gameplay Tag - Leaping:Gameplay Tag Structure)
   n7 = Get PlayerPawn() -> (PlayerPawn:Player Pawn Athena Object Reference)
   > n8 = WaitGameplayTagAddToActor!(execute:Exec, TargetActor:Actor Object Reference=@n7, Tag:Gameplay Tag Structure=@n6, OnlyTriggerOnce:bool=false) -> (then:Exec, AsyncTaskProxy:Ability Async Wait Gameplay Tag Added Object Reference, Added:Exec) {is=title K2Node_AsyncAction}
   n9 = Get Gameplay Tag - Flying() -> (Gameplay Tag - Flying:Gameplay Tag Structure)
   > n10 = WaitGameplayTagAddToActor(TargetActor=@n7, Tag=@n9, OnlyTriggerOnce=false)
   > n11 = WaitGameplayTagRemoveFromActor!(execute:Exec, TargetActor:Actor Object Reference=@n7, Tag:Gameplay Tag Structure=@n6, OnlyTriggerOnce:bool=false) -> (then:Exec, AsyncTaskProxy:Ability Async Wait Gameplay Tag Removed Object Reference, Removed:Exec) {is=title K2Node_AsyncAction}
   > n12 = WaitGameplayTagRemoveFromActor(TargetActor=@n7, Tag=@n9, OnlyTriggerOnce=false)
   n13 = Get isOnCooldown() -> (isOnCooldown:Gameplay Tag Structure)
   > n14 = WaitGameplayTagAddToActor(TargetActor=@n7, Tag=@n13, OnlyTriggerOnce=false)
   > n15 = WaitGameplayTagRemoveFromActor(TargetActor=@n7, Tag=@n13, OnlyTriggerOnce=false)
   n16 = IsLeapingOrFlying!(execute=@n15.Removed, self) -> (then, ReturnValue:bool)
   > n17 = Branch(Condition=@n16)
   n18 = Get OnCooldown() -> (OnCooldown:Widget Animation Object Reference)
   PlayAnimation(execute=@n17.else, InAnimation=@n18, PlayMode=Reverse, bRestoreState=true)
   n19 = IsLeapingOrFlying(execute=@n14.Added)
   > Branch(Condition=@n19)
   > PlayAnimation(InAnimation=@n18)
   Update Visibility(execute=[@n12.Removed, @n10.Added, @n11.Removed, @n8.Added])
   Event Construct()
   > n20 = Sequence()
   > n21 = Cast To PlayerPawn_Athena(Object=GetViewTarget!(self:Controller Object Reference=GetOwningPlayer()) -> (ReturnValue:Actor Object Reference))
   > Set PlayerPawn(PlayerPawn:Player Pawn Athena Object Reference=@n21) -> (Output_Get:Player Pawn Athena Object Reference)
   > n22 = GetDynamicMaterial(self=Get centerRet() -> (centerRet:Image Object Reference))
   > Set CenterRetMat(CenterRetMat:Material Instance Dynamic Object Reference=@n22) -> (Output_Get:Material Instance Dynamic Object Reference)
   n23 = GetEquippedWeapon!(self:Fort Weapon Reticle Extension Widget Base Object Reference) -> (ReturnValue:Fort Weapon Object Reference)
   Is Valid(exec=@n20.then_1, InputObject=@n23)
   > n24 = Set Weapon(Weapon:Fort Weapon Object Reference=@n23) -> (Output_Get:Fort Weapon Object Reference)
   > Bind Event to On Local Ammo Changed(self:Fort Weapon Object Reference=@n24, Delegate=Create Event()) {is=title K2Node_AddDelegate}
   > n25 = GetDynamicMaterial(self=Get fuelRet() -> (fuelRet:Image Object Reference))
   > Set FuelMat(FuelMat:Material Instance Dynamic Object Reference=@n25) -> (Output_Get:Material Instance Dynamic Object Reference)
   > n26 = Get()
   > n27 = Stop(self:Audio Component Object Reference=@n26)
   n28 = SpawnSound2D!(
    execute=[@n27.then, @n26.else]
    Sound:Sound Base Object Reference=/SunRoseFlyingGameplay/Sounds/MSS_SunRose_Wings_UI_LowEnergyPulse
    VolumeMultiplier:float=1
    PitchMultiplier:float=1
    StartTime:float=0
    ConcurrencySettings:Sound Concurrency Object Reference
    bPersistAcrossLevelTransition:bool=false
    bAutoDestroy:bool=true
   ) -> (then, ReturnValue:Audio Component Object Reference)
   > Set ReticleSoundLoop(ReticleSoundLoop:Audio Component Object Reference=@n28) -> (Output_Get:Audio Component Object Reference)
   n29 = Get Weapon() -> (Weapon:Fort Weapon Object Reference)
   > UpdateAmmo!(
    execute
    self
    LocalCount:int=GetLocalMagazineAmmoCount!(self:Fort Weapon Object Reference=@n29) -> (ReturnValue:int=0)
    LocalRemaining:int=GetLocalRemainingAmmo!(self:Fort Weapon Object Reference=@n29) -> (ReturnValue:int=0)
   ) -> (then) {is=title K2Node_CallFunction}
   > Update Visibility()
   > Show Ammo Hint()
   n30 = Get PlayerPawn()
   Is Valid(exec=@n20.then_2, InputObject=@n30)
   n31 = GetFortAbilitySystemComponent!(Actor:Actor Object Reference=@n30) -> (ReturnValue:Fort Ability System Component Object Reference)
   > n32 = Is Valid(InputObject=@n31)
   Bind Event to On Ability System Component Initialized(execute=@n32.Is Not Valid, self:Fort Player Pawn Object Reference=@n30, Delegate=Create Event()) {is=title K2Node_AddDelegate}
   HandleASCReady!(
    execute=@n32.Is Valid
    self
    AbilitySystemComponent:Fort Ability System Component Object Reference=@n31
    AffectedPawn:Fort Player Pawn Object Reference=Get PlayerPawn()
   ) -> (then) {is=title K2Node_CallFunction}
   Event Tick()
   n33 = GetLocalViewingPlayerController!(self:Pawn Object Reference=Get PlayerPawn()) -> (ReturnValue:Player Controller Object Reference)
   n34 = GetValueAtLevel!(
    Input:"Scalable Float Structure (by ref)"=Get Row_Crashing_LookPitchClamp_Max() -> (Row_Crashing_LookPitchClamp_Max:Scalable Float Structure)
    Level:float=0
   ) -> (ReturnValue:float)
   n35 = MapRangeClamped(
    Value=NormalizeAxis!(Angle:float=GetDesiredRotation!(self:Controller Object Reference=@n33) -> (ReturnValue_Roll:float, ReturnValue_Pitch:float, ReturnValue_Yaw:float).ReturnValue_Pitch) -> (ReturnValue:float)
    InRangeA=float * float(A=@n34, B=0.5)
    InRangeB=@n34
   )
   n36 = "Equal (Float)"(A=@n35, B=1)
   n37 = Get CenterRetMat() -> (CenterRetMat:Material Instance Dynamic Object Reference)
   > SetScalarParameterValue(
    self=@n37
    ParameterName=Anchor Tick Opacity
    Value=SelectFloat(A=SelectFloat(A=1, B=0.45, bPickA=@n36), B=0, bPickA="Not Equal (Float)"(A=@n35))
   )
   > SetScalarParameterValue(self=@n37, ParameterName=DA Tick Opacity, Value=Lerp(A=0, B=0.75, Alpha=@n35))
   > SetScalarParameterValue(self=@n37, ParameterName=yOffsetController, Value=Lerp(A=0, B=1, Alpha=@n35))
   > n38 = Branch(Condition="float >= float"(A=@n35, B=1))
   > Do Once(Reset=@n40.then)
   > PlayAnimationForward(InAnimation=Get LockedTicks() -> (LockedTicks:Widget Animation Object Reference))
   > n39 = PlayAnimationForward(InAnimation=Get LockedBrackets() -> (LockedBrackets:Widget Animation Object Reference))
   Do Once(execute=@n38.else, Reset=@n39.then)
   > PlayAnimationReverse(InAnimation=Get LockedTicks())
   > n40 = PlayAnimationForward(InAnimation=Get UnlockBrackets() -> (UnlockBrackets:Widget Animation Object Reference), PlaybackSpeed=0.5)
   Event Destruct()
   > n41 = Sequence!(execute) -> (then_0, then_1)
   n42 = Get PlayerPawn()
   > Is Valid(InputObject=@n42)
   > Unbind Event from On Ability System Component Initialized(self:Fort Player Pawn Object Reference=@n42, Delegate=Create Event()) {is=title K2Node_RemoveDelegate}
   Unbind Event from On Local Ammo Changed(execute=@n41.then_1, self:Fort Weapon Object Reference=Get Weapon(), Delegate=Create Event()) {is=title K2Node_RemoveDelegate}
   > n43 = Get()
   > Stop(self=@n43)
   Event OnInitialized()
   n44 = Get Gameplay Tag - Leaping()
   n45 = Get Gameplay Tag - Flying()
   > Set Gamplay Tag Array Data(Gamplay Tag Array Data:Gameplay Tag Container Structure=MakeGameplayTagContainerFromArray!(GameplayTags:Array of Gameplay Tag Structures=Make Array!("[0]"=@n44, "[1]":Gameplay Tag Structure=@n45) -> (Array)) -> (ReturnValue:Gameplay Tag Container Structure)) -> (Output_Get:Gameplay Tag Container Structure)
   > BindToAnimationFinished(Animation=Get AmmoHint() -> (AmmoHint:Widget Animation Object Reference), Delegate=Create Event())
  ]}
 IsLeapingOrFlying {nodes [
   IsLeapingOrFlying() {is=title K2Node_FunctionEntry}
   > n1 = HasAnyMatchingGameplayTags!(
    execute:Exec
    self:Object Reference=Get PlayerPawn()
    TagContainer:"Gameplay Tag Container Structure (by ref)"=Get Gamplay Tag Array Data() -> (Gamplay Tag Array Data:Gameplay Tag Container Structure)
   ) -> (then:Exec, ReturnValue:bool) {is=title K2Node_Message}
   > Return Node!(execute, ReturnValue:bool=@n1)
  ]}
 SetAudioComponentHidden {nodes [
   n1 = Inputs!() -> (In:Exec, Hidden:bool=~)
   > n2 = Get()
   > n3 = Set Boolean Parameter!(execute, self:Audio Parameter Controller Interface Interface=@n2, InName:Name=Hidden, InBool:bool=@n1) -> (then) {is=title K2Node_CallFunction}
   Outputs!(Out:Exec=[@n3.then, @n2.else])
  ]}
 Show Ammo Hint {nodes [
   Show Ammo Hint() {is=title K2Node_FunctionEntry}
   > n1 = IsLeapingOrFlying() {is=title K2Node_CallFunction}
   > n2 = Branch(Condition=@n1)
   SetVisibility(execute=@n2.else, self=Get fuelRet(), InVisibility=SelfHitTestInvisible)
   > PlayAnimation(InAnimation=Get AmmoHint())
  ]}
 Update Cooldown {nodes [Update Cooldown() {is=title K2Node_FunctionEntry}]}
 Update Visibility {nodes [
   Update Visibility() {is=title K2Node_FunctionEntry}
   > n1 = IsLeapingOrFlying()
   > SetVisibility(self=Get fuelRet(), InVisibility=Select(Option 0=Hidden, Option 1=SelfHitTestInvisible, Index=@n1))
   n2 = Select(Option 0:float=0, Option 1:float=1, Index=@n1) -> (ReturnValue:float)
   n3 = Get CenterRetMat()
   > SetScalarParameterValue(self=@n3, ParameterName=useDiveTicks, Value=@n2)
   > SetScalarParameterValue(self=@n3, ParameterName=useAnchorTicks, Value=@n2)
   > SetScalarParameterValue(self=@n3, ParameterName=useDiveBrackets, Value=@n2)
   > SetScalarParameterValue(self=Get FuelMat(), ParameterName=Fuel Opacity, Value=1)
   > Set Audio Component Hidden!(In:Exec, Hidden:bool=NOT Boolean(A=@n1)) -> (Out:Exec) {is=title K2Node_MacroInstance}
   > Branch(Condition=@n1)
   > StopAnimation(InAnimation=Get AmmoHint())
  ]}
]
variables [
 CenterRetMat {type=MaterialInstanceDynamic}
 FuelMat {type=MaterialInstanceDynamic}
 Gameplay Tag - Flying {type=GameplayTag, default {tagName=Granted.Athena.SunRose.Wings.Flying}}
 Gameplay Tag - Leaping {type=GameplayTag, default {tagName=Granted.Athena.SunRose.Wings.Leaping}}
 Gamplay Tag Array Data {type=GameplayTagContainer, default {gameplayTags []}}
 isOnCooldown {type=GameplayTag, default {tagName=Granted.Cooldown.Athena.SunRose.Wings.Leap}}
 Locked? {type=boolean}
 Only Trigger Once {type=boolean}
 PlayerPawn {type=PlayerPawn_Athena_C}
 ReticleSoundLoop {type=AudioComponent, default=null}
 Row_Crashing_LookPitchClamp_Max {
  type=ScalableFloat
  default {
   registryType=None
   value=1
   curve {
    curveTable=/SunRoseFlyingGameplay/DataTables/SunRoseFlyingGameplayGameData
    rowName=Default.SunRoseCrashingGameplay.Crashing.LookPitchClamp.Max
   }
  }
 }
 Weapon {type=FortWeapon, default=null}
]
widgets {
 info {
  inheritedWidgetCount=1
  parentClass=/Script/FortniteUI.FortWeaponReticleExtensionWidgetBase
  rootWidgetClass=&Overlay
  widgetCount=4
 }
 widgets [
  Overlay_15 {widgetClassPath=&Overlay, widget=Overlay_15}
  fuelRet {
   widgetClassPath=&Image
   bIsVariable=true
   parent=Overlay_15
   widget=fuelRet
   properties {
    visibility=Hidden
    brush {
     bIsDynamicallyLoaded=false
     drawAs=Image
     imageType=FullColor
     mirroring=NoMirror
     resourceName=None
     resourceObject=&/M_UI_SunroseReticle
     tiling=NoTile
     imageSize=<256 256>
     margin=<ltrb 0 0 0 0>
     outlineSettings {
      bUseBrushTransparency=false
      roundingType=HalfHeightRadius
      width=0
      color {colorUseRule=UseColor_Specified, specifiedColor=<rgba 0 0 0 0>}
      cornerRadii=<0 0 0 0>
     }
     tintColor {colorUseRule=UseColor_Specified, specifiedColor=<rgba 1 1 1 1>}
     uVRegion {bIsValid=false, max=<0 0>, min=<0 0>}
    }
    renderTransform {angle=0, scale=<0.75 0.75>, shear=<0 0>, translation=<0 25>}
   }
   slot {class=OverlaySlot, properties {horizontalAlignment=HAlign_Center, verticalAlignment=VAlign_Center}}
  }
  centerRet {
   widgetClassPath=&Image
   bIsVariable=true
   parent=Overlay_15
   widget=centerRet
   properties {brush {
     bIsDynamicallyLoaded=false
     drawAs=Image
     imageType=FullColor
     mirroring=NoMirror
     resourceName=None
     resourceObject=&/M_UI_UXReticle_Inst
     tiling=NoTile
     imageSize=<150 150>
     margin=<ltrb 0 0 0 0>
     outlineSettings {
      bUseBrushTransparency=false
      roundingType=HalfHeightRadius
      width=0
      color {colorUseRule=UseColor_Specified, specifiedColor=<rgba 0 0 0 0>}
      cornerRadii=<0 0 0 0>
     }
     tintColor {colorUseRule=UseColor_Specified, specifiedColor=<rgba 1 1 1 1>}
     uVRegion {bIsValid=false, max=<0 0>, min=<0 0>}
    }}
   slot {class=OverlaySlot, properties {horizontalAlignment=HAlign_Center, verticalAlignment=VAlign_Center}}
  }
  TutorialHighlight {widgetClassPath=&TutorialHighlight}
 ]
}
